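# Niagara impact effect for the Galaxy variant of the Crew pickaxe cosmetic.
# Galaxy_Smoke bursts a few short-lived, translucent cyan sprites from a small sphere, nudged by camera offset, that grow and fade over life.
# A front-end modifier fed by System.FrontendScale adjusts its spawn radius and size, and user booleans flag front-end and preview contexts.
# The other eight emitters (spikes, puff, flare, sparks, wave, dust) show no modules or renderers here.
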NiagaraSystem /BRCosmetics/Effects/Niagara/Systems/CosmeticTemplates/Pickaxe_Impacts/Crew/Galaxy/NS_Pickaxe_Crew_Impact_Galaxy
dependencies {
 usedDataInterfaces [&NiagaraDataInterfaceCurve, &NiagaraDataInterfaceVector2DCurve]
 usedDynamicInputs []
 usedModules [
  &self:SystemSpawnScript.SystemScriptSource.SystemScriptGraph.NiagaraNodeAssignment_0.SetVariables_#1
  &SystemState
 ]
}
emitters {
 Dust {rendererClasses [], renderers []}
 FlareAndLight {rendererClasses [], renderers []}
 Galaxy_Smoke {
  emitterUpdateScript [
   &EmitterState(Life Cycle Mode=::NewEnumerator1, Loop Behavior=::NewEnumerator1, Loop Duration=0.5, UseLoopDelay=true)
   &SpawnBurst_Instantaneous(
    Spawn Count=&RandomRangeInt(Minimum=5, Maximum=8)
    Use Spawn Probability=false
    UseLoopCountLimit=false
   )
  ]
  particleSpawnScript [
   &InitializeParticle(
    Color=&MakeLinearColorFromVectorAndFloat(
     "Vector (RGB)"=&MakeVectorFromLinearColorRGB(Color=<rgba 0.209 0.9868 1 0.5>)
     "Float (Alpha)"=&RandomRangeFloat(Minimum=0.25, Maximum=0.5)
    )
    Lifetime=&RandomRangeFloat(Minimum=0.3, Maximum=0.4)
    Sprite Rotation Mode=::NewEnumerator2
    Sprite Size Mode=::NewEnumerator3
    Uniform Sprite Size=&RandomRangeFloat(Minimum=85, Maximum=110)
    UsePositionOffset=false
    Write Material Random=true
   )
   &ShapeLocation(
    MaskSpawnGroup=false
    Override Seed=false
    Sphere Radius=&Multiply_Float(A=5, B=&FN_FrontEndFloatModifier(Gameplay=1, Front End=$System.FrontendScale))
    UseOwnerScale=true
   )
   &CameraOffset(Camera Offset Amount=10)
  ]
  particleUpdateScript [
   &ParticleState()
   &Drag(
    Drag=&FloatFromCurve(
     FloatCurve=DataInterface() {propertyValues {Curve {keys [
         {arriveTangent=0.8664, interpMode=RCIM_Cubic, leaveTangent=0.8664, tangentMode=RCTM_User, time=0, value=0.375}
         {interpMode=RCIM_Cubic, tangentMode=RCTM_User, time=1, value=1}
        ]}}}
     Scale Curve=&RandomRangeFloat(Minimum=12, Maximum=15)
    )
    Use Linear Drag=true
    Use Rotational Drag=false
   )
   &SolveForcesAndVelocity(Clamp Velocity=false, Limit Acceleration=false)
   &ScaleColor(
    Scale Alpha=&FloatFromCurve(FloatCurve=DataInterface() {propertyValues {Curve {keys [
         {arriveTangent=-0.0405, interpMode=RCIM_Cubic, leaveTangent=-0.0405, tangentMode=RCTM_User, time=0, value=1}
         {arriveTangent=-2.127, interpMode=RCIM_Cubic, leaveTangent=-2.127, tangentMode=RCTM_User, time=1, value=0}
        ]}}})
    Scale RGB=&VectorFromFloat(Value=&FloatFromCurve(
      FloatCurve=DataInterface() {propertyValues {Curve {keys [
          {arriveTangent=-2.4245, interpMode=RCIM_Cubic, leaveTangent=-2.4245, tangentMode=RCTM_User, time=0, value=1}
          {interpMode=RCIM_Cubic, tangentMode=RCTM_User, time=1, value=0}
         ]}}}
      Scale Curve=3
     ))
    ScaleA=true
    ScaleRGB=false
   )
   &ScaleSpriteSize(
    Uniform Curve Scale=&FN_FrontEndFloatModifier(Gameplay=1, Front End=$System.FrontendScale)
    Uniform Curve Sprite Scale=DataInterface() {propertyValues {Curve {keys [
        {arriveTangent=1.8336, interpMode=RCIM_Cubic, leaveTangent=1.8336, tangentMode=RCTM_User, time=0, value=0}
        {interpMode=RCIM_Cubic, tangentMode=RCTM_User, time=0.9, value=1}
       ]}}}
   )
  ]
  rendererClasses [&NiagaraSpriteRendererProperties]
  renderers [{rendererClass=&NiagaraSpriteRendererProperties}]
 }
 SmokePuff {rendererClasses [], renderers []}
 Sparks001 {rendererClasses [], renderers []}
 SpikeGreen {rendererClasses [], renderers []}
 SpikePink {rendererClasses [], renderers []}
 StarSpike {rendererClasses [], renderers []}
 Wave001 {rendererClasses [], renderers []}
}
summary {
 systemName=NS_Pickaxe_Crew_Impact_Galaxy
 emitters [
  {emitterName=Galaxy_Smoke, rendererClasses [&NiagaraSpriteRendererProperties]}
  {emitterName=SpikePink, rendererClasses []}
  {emitterName=SpikeGreen, rendererClasses []}
  {emitterName=StarSpike, rendererClasses []}
  {emitterName=SmokePuff, rendererClasses []}
  {emitterName=FlareAndLight, rendererClasses []}
  {emitterName=Sparks001, rendererClasses []}
  {emitterName=Wave001, rendererClasses []}
  {emitterName=Dust, rendererClasses []}
 ]
 userVariables [
  User.bIsFrontEnd {type=NiagaraBool}
  User.bIsFrontEndPreview {type=NiagaraBool}
 ]
}
